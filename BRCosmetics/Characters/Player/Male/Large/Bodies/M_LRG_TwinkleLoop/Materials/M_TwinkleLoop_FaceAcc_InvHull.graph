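# Unlit masked inverted-hull anime outline material for the TwinkleLoop character's face accessory.
# The shared outline function feeds a static switch: one branch applies the IndexEyes function, the other adds a camera-direction world position offset.
# That offset blends near and far amounts by FOV-compensated distance fade, scaled by a channel-selectable mask texture.
# The result passes through the CharacterParent PostFX function with pre-skinned position, distance blend and fresnel inputs.
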Material /BRCosmetics/Characters/Player/Male/Large/Bodies/M_LRG_TwinkleLoop/Materials/M_TwinkleLoop_FaceAcc_InvHull
graph {
 expressions {
  MaterialExpressionAdd_1 {
   class=MaterialExpressionAdd
   inputs [
    A {from=MaterialExpressionGetMaterialAttributes_0, output=World Position Offset}
    B {from=MaterialExpressionMaterialFunctionCall_5, output=Camera Offset}
   ]
  }
  MaterialExpressionChannelMaskParameter_0 {
   class=MaterialExpressionChannelMaskParameter
   inputs [None {from=MaterialExpressionTextureSampleParameter2D_0, output=RGB}]
   properties {group=OutlineOffset, parameterName=OutlineOffsetChannel}
  }
  MaterialExpressionComposite_0 {class=MaterialExpressionComposite, inputs [], properties {subgraphName=Construct Base Inputs}}
  MaterialExpressionConstant_1 {class=MaterialExpressionConstant}
  MaterialExpressionFresnel_0 {
   class=MaterialExpressionFresnel
   inputs [
    ExponentIn {from="", output=""}
    BaseReflectFractionIn {from="", output=""}
    Normal {from="", output=""}
   ]
   properties {baseReflectFraction=0, exponent=1}
  }
  MaterialExpressionGetMaterialAttributes_0 {
   class=MaterialExpressionGetMaterialAttributes
   inputs [None {from=MaterialExpressionMaterialFunctionCall_0, output=Result}]
   properties {attributeGetTypes [#1]}
  }
  MaterialExpressionLinearInterpolate_0 {
   class=MaterialExpressionLinearInterpolate
   inputs [
    A {from=MaterialExpressionScalarParameter_1}
    B {from=MaterialExpressionScalarParameter_0}
    Alpha {from=MaterialExpressionNamedRerouteUsage_0}
   ]
  }
  MaterialExpressionMaterialFunctionCall_0 {
   class=MaterialExpressionMaterialFunctionCall
   inputs []
   properties {materialFunction=/Game/Global/Materials/Character_Parent_2023/Anime_Parent_2023/MaterialFunctions/MFC_AnimeOutline_InvHull_2023}
  }
  MaterialExpressionMaterialFunctionCall_1 {
   class=MaterialExpressionMaterialFunctionCall
   inputs [
    Fade Offset {from=MaterialExpressionScalarParameter_39}
    Fade Distance {from=MaterialExpressionScalarParameter_38}
   ]
   properties {materialFunction=/Game/Global/Materials/Character_Parent_2023/Anime_Parent_2023/MaterialFunctions/MFC_DistanceFade_FOV_compensation}
  }
  MaterialExpressionMaterialFunctionCall_14 {
   class=MaterialExpressionMaterialFunctionCall
   inputs [Material {from=""}]
   properties {materialFunction=/Game/Packages/Fortress_SharedMaterials/Base_Material_Functions/CharacterBrightnessandScale_Balance}
  }
  MaterialExpressionMaterialFunctionCall_16 {
   class=MaterialExpressionMaterialFunctionCall
   inputs [
    Material {from=MaterialExpressionStaticSwitchParameter_0}
    M {from=MaterialExpressionReroute_8, output=M}
    Pre Skinned Local Position {from=MaterialExpressionReroute_9, output=Pre Skinned Local Position}
    Distance Blend {from=MaterialExpressionReroute_10, output=Distance Blend}
    Fresnel {from=MaterialExpressionReroute_11, output=Fresnel}
   ]
   properties {materialFunction=&MF_CharacterParent_2023_PostFX}
  }
  MaterialExpressionMaterialFunctionCall_4 {
   class=MaterialExpressionMaterialFunctionCall
   inputs [Material {from=MaterialExpressionMaterialFunctionCall_0, output=Result}]
   properties {materialFunction=/Game/Characters/Player/Common/MaterialFunctions/VertexColorFaces/MF_FV2_IndexEyes}
  }
  MaterialExpressionMaterialFunctionCall_5 {
   class=MaterialExpressionMaterialFunctionCall
   inputs [
    Offset Amount {from=MaterialExpressionMultiply_0}
    Clamp Padding {from=""}
    World Position {from=""}
   ]
   properties {materialFunction=/Engine/Functions/Engine_MaterialFunctions02/WorldPositionOffset/CameraOffset}
  }
  MaterialExpressionMultiply_0 {
   class=MaterialExpressionMultiply
   inputs [
    A {from=MaterialExpressionLinearInterpolate_0}
    B {from=MaterialExpressionChannelMaskParameter_0, output=R}
   ]
  }
  MaterialExpressionNamedRerouteDeclaration_0 {
   class=MaterialExpressionNamedRerouteDeclaration
   inputs [None {from=MaterialExpressionMaterialFunctionCall_1, output=Result}]
   properties {name=DistanceFading, nodeColor=<rgba 1 0.3047 0 1>}
  }
  MaterialExpressionNamedRerouteUsage_0 {class=MaterialExpressionNamedRerouteUsage}
  MaterialExpressionPinBase_0 {class=MaterialExpressionPinBase, inputs [], properties {}}
  MaterialExpressionPinBase_1 {
   class=MaterialExpressionPinBase
   inputs [
    M {from=MaterialExpressionConstant_1, output=""}
    Pre Skinned Local Position {from=MaterialExpressionShaderStageSwitch_0, output=""}
    Distance Blend {from=MaterialExpressionMaterialFunctionCall_14, output=Distance Blend}
    Fresnel {from=MaterialExpressionFresnel_0, output=""}
   ]
   properties {reroutePins [
     M {expression=&self:MaterialGraph_1.MaterialGraphNode_Composite_0.Construct Base Inputs.MaterialExpressionReroute_8}
     Pre Skinned Local Position {expression=&self:MaterialGraph_1.MaterialGraphNode_Composite_0.Construct Base Inputs.MaterialExpressionReroute_9}
     Distance Blend {expression=&self:MaterialGraph_1.MaterialGraphNode_Composite_0.Construct Base Inputs.MaterialExpressionReroute_10}
     Fresnel {expression=&self:MaterialGraph_1.MaterialGraphNode_Composite_0.Construct Base Inputs.MaterialExpressionReroute_11}
    ]}
  }
  MaterialExpressionPreSkinnedPosition_0 {class=MaterialExpressionPreSkinnedPosition, inputs [], properties {}}
  MaterialExpressionScalarParameter_0 {class=MaterialExpressionScalarParameter, properties {defaultValue=-1.5, group=OutlineOffset, parameterName=OffsetAmount_Far}}
  MaterialExpressionScalarParameter_1 {class=MaterialExpressionScalarParameter, properties {group=OutlineOffset, parameterName=OffsetAmount}}
  MaterialExpressionScalarParameter_38 {
   class=MaterialExpressionScalarParameter
   properties {defaultValue=1000, group=01_InvertedHull_Outlines, parameterName=Fade_Distance, sliderMax=1500, sliderMin=500, sortPriority=5}
  }
  MaterialExpressionScalarParameter_39 {
   class=MaterialExpressionScalarParameter
   properties {defaultValue=50, group=01_InvertedHull_Outlines, parameterName=Fade_Offset, sliderMax=100, sortPriority=5}
  }
  MaterialExpressionSetMaterialAttributes_1 {
   class=MaterialExpressionSetMaterialAttributes
   inputs [
    MaterialAttributes {from=MaterialExpressionGetMaterialAttributes_0, output=MaterialAttributes}
    World Position Offset {from=MaterialExpressionAdd_1, output=""}
   ]
   properties {attributeSetTypes [#1]}
  }
  MaterialExpressionShaderStageSwitch_0 {
   class=MaterialExpressionShaderStageSwitch
   inputs [
    PixelShader {from=MaterialExpressionVertexInterpolator_0, output=PS}
    VertexShader {from=MaterialExpressionPreSkinnedPosition_0, output=""}
   ]
   properties {}
  }
  MaterialExpressionStaticSwitchParameter_0 {
   class=MaterialExpressionStaticSwitchParameter
   inputs [
    True {from=MaterialExpressionSetMaterialAttributes_1, output=""}
    False {from=MaterialExpressionMaterialFunctionCall_4, output=Result}
   ]
   properties {group=OutlineOffset, parameterName=UseOutlineOffsetInCamDir, sortPriority=0}
  }
  MaterialExpressionTextureSampleParameter2D_0 {
   class=MaterialExpressionTextureSampleParameter2D
   inputs [
    UVs {from="", output=""}
    Apply View MipBias {from="", output=""}
   ]
   properties {group=OutlineOffset, parameterName=OutlineOffsetMask, texture=&T_BlacksRGB}
  }
  MaterialExpressionVertexInterpolator_0 {class=MaterialExpressionVertexInterpolator, inputs [VS {from=MaterialExpressionPreSkinnedPosition_0, output=""}], properties {}}
 }
 properties {
  bAutomaticallySetUsageInEditor=false
  bEnableResponsiveAA=true
  bUseMaterialAttributes=true
  bUsedWithMeshDeformer=true
  bUsedWithMorphTargets=true
  bUsedWithSkeletalMesh=true
  blendMode=BLEND_Masked
  pixelDepthOffsetMode=PDOM_Legacy
  shadingModel=MSM_Unlit
  twoSided=true
  usedShadingModels=Unlit
 }
}
