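# Animation Blueprint for the face accessory part of the TwinkleLoop large male character, driving flipbook eyebrows on an overlay material.
# The AnimGraph copies pose and curves from the attached parent through the head's Control Rig node, whose alpha is zero.
# On customization load, it caches a dynamic overlay material from the part mesh.
# Each update, after a zero-length delay, it writes left and right brow curve values to that material's flipbook index parameters.
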AnimBlueprint /BRCosmetics/Characters/Player/Male/Large/Bodies/M_LRG_TwinkleLoop/Meshes/Parts/M_LRG_TwinkleLoop_FaceAcc_AnimBP
target_skeleton=&/M_LRG_TwinkleLoop_FaceAcc_Skeleton
graphs [
 AnimGraph {nodes [
   n1 = Control Rig(Source=Copy Pose From Mesh() {bCopyCurves=true, bUseAttachedParent=true}, Alpha=0) {
    alphaBoolBlend {blendOption=Linear}
    controlRigAssetReference {blueprintRigClass=/BRCosmetics/Characters/Player/Male/Large/Heads/M_LRG_TwinkleLoop_Head/Meshes/M_LRG_TwinkleLoop_Head_CtrlRig.M_LRG_TwinkleLoop_Head_CtrlRig_C}
   }
   Output Pose(Result=@n1) {layerGroup=None}
  ]}
 EventGraph {nodes [
   Event BlueprintUpdateAnimation()
   > Delay(Duration=0)
   > SetScalarParameterValue(
    self=Get overlay_mat() -> (overlay_mat:Material Instance Dynamic Object Reference)
    ParameterName=flipbook_face_L_brow_index
    Value=GetCurveValue(CurveName=flipbook_face_L_brow_index)
   )
   > SetScalarParameterValue(self=Get overlay_mat(), ParameterName=flipbook_face_R_brow_index, Value=GetCurveValue(CurveName=flipbook_face_R_brow_index))
   Event OnCharacterCustomizationLoaded()
   > n1 = GetOrCreateDynamicOverlayMaterial(MeshComponent=GetPartSkeletalMeshComponent(), MaterialName=None)
   > Set overlay_mat(overlay_mat:Material Instance Dynamic Object Reference=@n1) -> (Output_Get:Material Instance Dynamic Object Reference)
   Event BlueprintInitializeAnimation()
  ]}
]
variables [overlay_mat {type=MaterialInstanceDynamic}]
